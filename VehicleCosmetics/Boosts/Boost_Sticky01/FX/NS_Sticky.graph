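# Niagara system for the Sticky vehicle boost cosmetic, drawing an exhaust effect with web-named material parameters behind the car.
# ConeOptimized merges five spawn groups into one emitter, selecting lifetime, scale and rotation per group for six mesh renderers and a sprite renderer.
# Optimized_Trail adds a widening ribbon, while RibbonN001 adds two ribbons pushed sideways, both inheriting source movement with falloff over life.
# User parameters drive placement, game-mode ribbon rate and velocity, and frontend-preview brightness.
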NiagaraSystem /VehicleCosmetics/Boosts/Boost_Sticky01/FX/NS_Sticky
dependencies {
 usedDataInterfaces [
  &NiagaraDataInterfaceCurve
  &NiagaraDataInterfaceVectorCurve
  &NiagaraDataInterfaceSkeletalMesh
  &NiagaraDataInterfaceArrayFloat3
 ]
 usedDynamicInputs [&SelectVectorFromArray]
 usedModules [
  &self:SystemSpawnScript.SystemScriptSource.SystemScriptGraph.NiagaraNodeAssignment_0.SetVariables_#1
  &self:SystemSpawnScript.SystemScriptSource.SystemScriptGraph.NiagaraNodeAssignment_1.SetVariables_#2
  &SystemState
  &self:SystemSpawnScript.SystemScriptSource.SystemScriptGraph.NiagaraNodeAssignment_2.SetVariables_#3
 ]
}
emitters {
 ConeOptimized {
  emitterUpdateScript [
   &EmitterState()
   &VC_FilteredSocketCount() {enabled=false}
   &SpawnRate(SpawnRate=14, Use Spawn Probability=false)
   &SpawnRate(SpawnGroup=1, SpawnRate=14, Use Spawn Probability=false) {module=SpawnRate001}
   &SpawnRate(SpawnGroup=2, SpawnRate=30, Use Spawn Probability=false) {module=SpawnRate002}
   &SpawnRate(SpawnGroup=3, SpawnRate=3, Use Spawn Probability=false) {module=SpawnRate003}
   &SpawnRate(SpawnGroup=4, SpawnRate=8, Use Spawn Probability=false) {module=SpawnRate004}
  ]
  particleSpawnScript [
   &VC_PaintedVariantsOverride() {enabled=false}
   /SetVariables(Particles.NewVisTag:NiagaraInt32=&SpawnGroup()) {isSetParameters=true}
   &InitializeParticle(
    Color Mode=::NewEnumerator0
    Lifetime Max=/VehicleCosmetics/FX/Niagara/Modules/MergeEmitters/VC_FloatBySpawnGroup(
     SpawnGroup or ExecIndex:NiagaraInt32=0
     CustomInt:NiagaraInt32=0
     MaximumSpawnGroups:NiagaraInt32=5
     Float SG 0:NiagaraFloat=0.58
     Float SG 1:NiagaraFloat=0.7
     Float SG 2:NiagaraFloat=0.25
     Float SG 3:NiagaraFloat=1.6
     Float SG 4:NiagaraFloat=0.8
     Float SG 5:NiagaraFloat=0
    )
    Lifetime Min:NiagaraFloat=&VC_FloatBySpawnGroup(
     SpawnGroup or ExecIndex:NiagaraInt32=0
     CustomInt:NiagaraInt32=0
     MaximumSpawnGroups:NiagaraInt32=5
     Float SG 0:NiagaraFloat=0.4
     Float SG 1:NiagaraFloat=0.45
     Float SG 2:NiagaraFloat=0.15
     Float SG 3:NiagaraFloat=1
     Float SG 4:NiagaraFloat=0.5
     Float SG 5:NiagaraFloat=0
    )
    Lifetime Mode=::NewEnumerator1
    Mesh Scale Max=/VehicleCosmetics/FX/Niagara/Modules/MergeEmitters/VC_VectorBySpawnGroup(
     SpawnGroup or ExecIndex:NiagaraInt32=0
     CustomInt:NiagaraInt32=0
     MaximumSpawnGroups:NiagaraInt32=5
     Vector Group 0:Vector3f=<0.45 0.45 0.75>
     Vector Group 1:Vector3f=<0.21 4.8 3.6>
     Vector Group 2:Vector3f=<3 1.65 1.65>
     Vector Group 3:Vector3f=&VectorFromFloat(Value=1.08)
     Vector Group 4:Vector3f=<1.38 0.8625 0.8625>
     Vector Group 5:Vector3f=<0 0 0>
    )
    Mesh Scale Min=&VC_VectorBySpawnGroup(
     SpawnGroup or ExecIndex:NiagaraInt32=0
     CustomInt:NiagaraInt32=0
     MaximumSpawnGroups:NiagaraInt32=5
     Vector Group 0:Vector3f=<0.15 0.15 0.54>
     Vector Group 1:Vector3f=<2.1 3.6 4.8>
     Vector Group 2:Vector3f=<3 1.05 1.05>
     Vector Group 3:Vector3f=&VectorFromFloat(Value=0.8)
     Vector Group 4:Vector3f=<1.38 0.8625 0.8625>
     Vector Group 5:Vector3f=<0 0 0>
    )
    Mesh Scale Mode=::NewEnumerator2
    Position Offset=&AddVectorToPosition(
     Position=&MergeHelper_VectorSwitch(0=<-200 0 32>, 1=<-200 0 32>, 2=<-200 0 32>, 3=<-200 0 32>, Index=$User.GameModeID, Num=::NewEnumerator2)
     Vector=&AddVectorToPosition(Position=$User.DTCarOffset, Vector=$User.SingleBoostPos)
    )
    Position Offset Coordinate Space=::Local
    Sprite Rotation Mode=::NewEnumerator2
    UsePositionOffset=true
    Write Material Random=true
   )
   &PartitionParticles(Number Of Partitions=$Emitter.PortCount) {enabled=false}
   &SkeletalMeshLocation(
    Bone / Socket Index=$Particles.PartitionParticles.Partition
    Bone / Socket Sampling Mode=::NewEnumerator5
    MaskSpawnGroup=false
    Mesh Sampling Type=::NewEnumerator6
    Skeletal Mesh=$System.SkeletalMesh
   ) {enabled=false}
   &AddVelocity(Velocity=<-1500 0 0>) {enabled=false}
   &InitialMeshOrientation(
    Mesh Orientation Mode=::NewEnumerator7
    Rotation=&RandomRangeVector(
     Minimum=&VC_VectorBySpawnGroup(
      SpawnGroup or ExecIndex:NiagaraInt32=0
      CustomInt:NiagaraInt32=0
      MaximumSpawnGroups:NiagaraInt32=5
      Vector Group 0:Vector3f=<0 0 0.2>
      Vector Group 1:Vector3f=<0 0 0>
      Vector Group 2:Vector3f=<0 0 0.05>
      Vector Group 3:Vector3f=<1 0 0.01>
      Vector Group 4:Vector3f=<0 0 0>
      Vector Group 5:Vector3f=<0 0 0>
     )
     Maximum=&VC_VectorBySpawnGroup(
      SpawnGroup or ExecIndex:NiagaraInt32=0
      CustomInt:NiagaraInt32=0
      MaximumSpawnGroups:NiagaraInt32=5
      Vector Group 0:Vector3f=<-1 0 1>
      Vector Group 1:Vector3f=<1 0 0>
      Vector Group 2:Vector3f=<1 0 0.05>
      Vector Group 3:Vector3f=<1 0 0.01>
      Vector Group 4:Vector3f=<0 0 0>
      Vector Group 5:Vector3f=<0 0 0>
     )
    )
    Use Rotation Vector=true
   )
   &Color()
  ]
  particleUpdateScript [
   &ParticleState()
   &UpdateMeshOrientation(Rotation Vector=&VC_VectorBySpawnGroup(
     SpawnGroup or ExecIndex:NiagaraInt32=0
     CustomInt:NiagaraInt32=0
     MaximumSpawnGroups:NiagaraInt32=5
     Vector Group 0:Vector3f=<0 0 0>
     Vector Group 1:Vector3f=<0 0 0>
     Vector Group 2:Vector3f=<0 0 0>
     Vector Group 3:Vector3f=<2.75 0 0>
     Vector Group 4:Vector3f=&RandomRangeVector(Minimum=<0.1 0 0>, Maximum=<0.2 0 0>)
     Vector Group 5:Vector3f=<0 0 0>
    ))
   &ScaleMeshSize(Scale Factor=&VC_VectorBySpawnGroup(
     SpawnGroup or ExecIndex:NiagaraInt32=0
     CustomInt:NiagaraInt32=0
     MaximumSpawnGroups:NiagaraInt32=5
     Vector Group 0:Vector3f=&VectorFromFloat(Value=&FloatFromCurve(FloatCurve=DataInterface() {propertyValues {Curve {keys [
           {interpMode=RCIM_Cubic, tangentMode=RCTM_Break, time=0, value=0.4}
           {interpMode=RCIM_Cubic, leaveTangent=1.2138, tangentMode=RCTM_Break, time=0.4, value=1.31}
           {arriveTangent=-0.5403, interpMode=RCIM_Cubic, tangentMode=RCTM_Break, time=1, value=1.76}
          ]}}}))
     Vector Group 1:Vector3f=&VectorFromFloat(Value=&FloatFromCurve(FloatCurve=DataInterface() {propertyValues {Curve {keys [
           {interpMode=RCIM_Cubic, leaveTangent=1.0479, tangentMode=RCTM_Break, time=0, value=0.4}
           {arriveTangent=0.3668, interpMode=RCIM_Cubic, leaveTangent=0.7617, tangentMode=RCTM_Break, time=0.37, value=0.7}
           {arriveTangent=1.9648, interpMode=RCIM_Cubic, tangentMode=RCTM_Break, time=0.82, value=1.14}
           {arriveTangent=7.2088, interpMode=RCIM_Cubic, tangentMode=RCTM_Break, time=1, value=1.74}
          ]}}}))
     Vector Group 2:Vector3f=&VectorFromFloat(Value=&FloatFromCurve(FloatCurve=DataInterface() {propertyValues {Curve {keys [
           {interpMode=RCIM_Cubic, tangentMode=RCTM_Break, time=0, value=0.4}
           {interpMode=RCIM_Cubic, tangentMode=RCTM_Break, time=0.17, value=1.25}
           {arriveTangent=1.8839, interpMode=RCIM_Cubic, tangentMode=RCTM_Break, time=1, value=1.92}
          ]}}}))
     Vector Group 3:Vector3f=&VectorFromFloat(Value=&FloatFromCurve(FloatCurve=DataInterface() {propertyValues {Curve {keys [<at 0 0.4>,<at 0.5 1>,<at 0.8 0>]}}}))
     Vector Group 4:Vector3f=<0 0 0>
     Vector Group 5:Vector3f=<0 0 0>
    ))
   &SolveForcesAndVelocity(Clamp Velocity=false, Limit Acceleration=false)
   &ScaleColor(Scale Alpha=20, Scale RGB=&Multiply_VectorByFloat(Vector=<1 1 1>, Float=1), ScaleA=true, ScaleRGB=true)
   &DynamicMaterialParameters(
    Index 0 Param 1=&FloatFromCurve(FloatCurve=DataInterface() {propertyValues {Curve {keys [{tangentMode=RCTM_User, time=0.7, value=0}, {tangentMode=RCTM_User, time=1, value=1}]}}})
    Index 0 Param 2=2
    Module.Index0Param0Write=true
    Module.Index0Param1Write=true
    Module.Index0Param2Write=true
    Module.Index0Param3Write=true
   )
   /SetVariables(
    Particles.DynamicCompleteString02:Vector4f=&MakeVector4(
     X=&FloatFromCurve(FloatCurve=DataInterface() {propertyValues {Curve {keys [<at 0 0.25>,<at 0.7 0.35>,<at 1 1>]}}})
     Y=1.45
     Z=0
     W=0
    )
    Particles.DynamicCone02:Vector4f=&MakeVector4(
     X=&FloatFromCurve(FloatCurve=DataInterface() {propertyValues {Curve {keys [<at 0.58 0>,<at 1 1>]}}})
     Y=2
     Z=&FloatFromCurve(FloatCurve=DataInterface() {propertyValues {Curve {keys [<at 0.5 0>,<at 1 5>]}}})
     W=0
    )
    Particles.DynamicWebRotation:Vector4f=&MakeVector4(
     X=&FloatFromCurve(FloatCurve=DataInterface() {propertyValues {Curve {keys [{interpMode=RCIM_Cubic, tangentMode=RCTM_Break, time=0.35, value=0}, {interpMode=RCIM_Cubic, tangentMode=RCTM_Break, time=1, value=0.6}]}}})
     Y=0
     Z=0
     W=0
    )
    Particles.DynamicWebVortex:Vector4f=&MakeVector4(
     X=&FloatFromCurve(FloatCurve=DataInterface() {propertyValues {Curve {keys [<at 0.7 0>,<at 1 1>]}}})
     Y=2
     Z=0
     W=0
    )
   ) {isSetParameters=true}
   /SetVariables(
    Particles.ColorCone02:LinearColor=/Niagara/DynamicInputs/LinearColor/MakeLinearColorFromFloat4(Vector:Vector4f=&MakeVector4(
      X=1
      Y=1
      Z=1
      W=&FloatFromCurve(FloatCurve=DataInterface() {propertyValues {Curve {keys [<at 0.65 1>,<at 1 0>]}}}, Scale Curve=20)
     ))
    Particles.ColorWebVortex:LinearColor=&MakeLinearColorFromFloat4(Vector:Vector4f=&MakeVector4(
      X=1
      Y=1
      Z=1
      W=&FloatFromCurve(FloatCurve=DataInterface() {propertyValues {Curve {keys [<at 0.4 1>,<at 1 0>]}}}, Scale Curve=20)
     ))
    Particles.ScaleMeshCone02:Vector3f=&VectorFromCurve(VectorCurve=DataInterface() {propertyValues {
       XCurve {keys [<at 0 0.3>,<at 0.5 1>,<at 0.55 1>,<at 0.8 0.6>]}
       YCurve {keys [<at 0 0.3>,<at 0.5 1>,<at 0.55 1>,<at 0.8 4>]}
       ZCurve {keys [<at 0 0.3>,<at 0.5 1>,<at 0.55 1>,<at 0.8 0.6>]}
      }})
    Particles.ScaleMeshCone03:Vector3f=&VectorFromCurve(VectorCurve=DataInterface() {propertyValues {
       XCurve {keys [<at 0 0.3>,<at 0.5 1>,<at 0.55 1>,<at 0.8 0.6>]}
       YCurve {keys [<at 0 0.3>,<at 0.5 1>,<at 0.55 1>,<at 0.8 0.6>]}
       ZCurve {keys [<at 0 0.3>,<at 0.5 1>,<at 0.55 1>,<at 0.8 4>]}
      }})
   ) {isSetParameters=true}
  ]
  rendererClasses [&NiagaraMeshRendererProperties, &NiagaraSpriteRendererProperties]
  renderers [
   {rendererClass=&NiagaraMeshRendererProperties, rendererIndex=0}
   {rendererClass=&NiagaraMeshRendererProperties, rendererIndex=1}
   {rendererClass=&NiagaraMeshRendererProperties, rendererIndex=2}
   {rendererClass=&NiagaraMeshRendererProperties, rendererIndex=3}
   {rendererClass=&NiagaraMeshRendererProperties, rendererIndex=4}
   {rendererClass=&NiagaraMeshRendererProperties, rendererIndex=5}
   {rendererClass=&NiagaraSpriteRendererProperties, rendererIndex=6}
  ]
 }
 Optimized_Trail {
  emitterUpdateScript [
   &EmitterState()
   &VC_FilteredSocketCount() {enabled=false}
   &VC_RateQualityHandler(Default Rate:NiagaraFloat=&MergeHelper_FloatSwitch(0=40, 1=40, 2=40, 3=10, Index=$User.GameModeID, Num=::NewEnumerator2))
   &VC_SpawnRibbon_v1(SpawnRate=$Emitter.Spawn Rate)
  ]
  particleSpawnScript [
   &VC_PaintedVariantsOverride() {enabled=false}
   &VC_InitializeRibbon_v1(
    Color=<rgba 1 1 1 500>
    Custom Offset=true
    Lifetime=1
    Local Vector Offset=&AddVectorToPosition(
     Position=&MergeHelper_VectorSwitch(0=<-220 0 32>, 1=<-220 0 32>, 2=<-220 0 32>, 3=<-210 0 32>, Index=$User.GameModeID, Num=::NewEnumerator2)
     Vector=&AddVectorToPosition(Position=$User.DTCarOffset, Vector=$User.SingleBoostPos)
    )
    Offset Amount=1
    Ribbon Width=45
   )
   &AddVelocity(Velocity=&MergeHelper_VectorSwitch(0=<-800 0 0>, 1=<-800 0 0>, 2=<-800 0 0>, 3=<-2400 0 0>, Index=$User.GameModeID, Num=::NewEnumerator2)) {module=AddVelocity001}
   &Color(Scale Alpha=&MakeCustomFloatFromBool(Bool=$User.bIsFrontendPreview, True Float=0.15, False Float=0.2))
   /SetVariables(Particles.Width2Ribbon=35) {isSetParameters=true}
   /VehicleCosmetics/FX/Niagara/Modules/MergeEmitters/VC_ExtraBindingColors(
    ExtraColors=3
    R1 Color:LinearColor=&MakeCustomLinearColorFromBool(Bool=$User.bIsFrontendPreview, True Color=<rgba 1 1 1 1>, False Color=<rgba 1 1 1 3>)
   )
   /VehicleCosmetics/FX/VC_FXParents/Modules/DLR_ProductPreviews(GameModeID:NiagaraInt32=$System.GameModeID, bPreviewVehicleSpeed:NiagaraBool=$System.bPreviewVehicleSpeed)
  ]
  particleUpdateScript [
   &ParticleState()
   &InheritSourceMovement(
    Applied Position Scale=&Multiply_VectorByFloat(
     Vector=<1 1 1>
     Float=&FloatFromCurve(FloatCurve=DataInterface() {propertyValues {Curve {keys [{interpMode=RCIM_Cubic, tangentMode=RCTM_Break, time=0.3, value=1}, {interpMode=RCIM_Cubic, tangentMode=RCTM_Break, time=1, value=0}]}}})
    )
    ApplyForce=false
    ApplyPosition=true
    ApplyPositionFalloff=false
   ) {enabled=false, module=InheritSourceMovement001}
   &InheritSourceMovement(
    Applied Position Scale=&VectorFromFloat(Value=&FloatFromCurve(FloatCurve=DataInterface() {propertyValues {Curve {
         defaultValue=1
         keys [
          {arriveTangent=-1.7885, interpMode=RCIM_Cubic, leaveTangent=-1.7885, tangentMode=RCTM_User, time=0.05, value=1}
          {interpMode=RCIM_Cubic, tangentMode=RCTM_User, time=1, value=0}
         ]
        }}}))
    ApplyForce=false
    ApplyPosition=true
    ApplyPositionFalloff=false
   )
   &SolveForcesAndVelocity(Clamp Velocity=false, Limit Acceleration=false)
   &ScaleRibbonWidth(Ribbon Width Scale=&FloatFromCurve(FloatCurve=DataInterface() {propertyValues {Curve {keys [{interpMode=RCIM_Cubic, tangentMode=RCTM_Break, time=0, value=0.88}, {interpMode=RCIM_Cubic, tangentMode=RCTM_Break, time=0.7, value=7}]}}}))
   &ScaleColor(
    Scale Alpha=&FloatFromCurve(FloatCurve=DataInterface() {propertyValues {Curve {keys [
         {interpMode=RCIM_Cubic, tangentMode=RCTM_Break, time=0, value=0}
         {interpMode=RCIM_Cubic, tangentMode=RCTM_Break, time=0.2, value=1}
         {interpMode=RCIM_Cubic, tangentMode=RCTM_Break, time=0.5, value=1}
         {interpMode=RCIM_Cubic, tangentMode=RCTM_Break, time=1, value=0}
        ]}}})
    ScaleA=true
    ScaleRGB=true
   )
   /SetVariables(Particles.R1 Color:LinearColor=&VC_DynamicScaleColor(
     Color to Scale:LinearColor=&MakeCustomLinearColorFromBool(Bool=$User.bIsFrontendPreview, True Color=<rgba 1 1 1 1>, False Color=<rgba 1 1 1 3>)
     RGB Scale:Vector3f=<3 3 3>
     Alpha Scale:NiagaraFloat=1
    )) {isSetParameters=true}
   &DynamicMaterialParameters(
    Index 0 Param 1=&FloatFromCurve(FloatCurve=DataInterface() {propertyValues {Curve {keys [<at 0 0.05>,<at 0.3 0>,<at 0.58 2.5>]}}})
    Index 0 Param 2=&FloatFromCurve(FloatCurve=DataInterface() {propertyValues {Curve {keys [<at 0.2 3>,<at 1 2.5>]}}})
    Module.Index0Float0Write=true
    Module.Index0Float1Write=true
    Module.Index0Float2Write=true
    Module.Index0Float3Write=true
   )
   /SetVariables(Particles.DynamicParameter1Bin:Vector4f=&MakeVector4(X=2.5, Y=-5, Z=-1.9, W=0)) {isSetParameters=true}
   &VC_UpdateRibbon_v1()
  ]
  rendererClasses [&NiagaraRibbonRendererProperties]
  renderers [
   {rendererClass=&NiagaraRibbonRendererProperties}
   {rendererClass=&NiagaraRibbonRendererProperties, rendererIndex=1}
  ]
 }
 RibbonN001 {
  emitterUpdateScript [
   &EmitterState()
   &VC_FilteredSocketCount() {enabled=false}
   &VC_RateQualityHandler(Default Rate:NiagaraFloat=&MergeHelper_FloatSwitch(0=40, 1=40, 2=40, 3=10, Index=$User.GameModeID, Num=::NewEnumerator2))
   &VC_SpawnRibbon_v1(Number of Ribbons=2, SpawnRate=$Emitter.Spawn Rate)
  ]
  particleSpawnScript [
   &VC_PaintedVariantsOverride() {enabled=false}
   &VC_InitializeRibbon_v1(
    Custom Offset=true
    Lifetime=1
    Local Vector Offset=&AddVectorToPosition(
     Position=&MergeHelper_VectorSwitch(0=<-220 0 32>, 1=<-220 0 32>, 2=<-220 0 32>, 3=<-210 0 32>, Index=$User.GameModeID, Num=::NewEnumerator2)
     Vector=&AddVectorToPosition(Position=$User.DTCarOffset, Vector=$User.SingleBoostPos)
    )
    Offset Amount=1
    Ribbon Width:NiagaraFloat=&Multiply_Float(A=5, B=7)
   )
   &AddVelocity(Velocity=&MergeHelper_VectorSwitch(0=<-800 0 0>, 1=<-800 0 0>, 2=<-800 0 0>, 3=<-4000 0 0>, Index=$User.GameModeID, Num=::NewEnumerator2)) {module=AddVelocity001}
   &Color(Scale Color=&VectorFromFloat(Value=&MakeCustomFloatFromBool(Bool=$User.bIsFrontendPreview, True Float=1, False Float=3)))
   &AddVelocity(Velocity=&MaskVectorBySpawnGroup(
     Masked Vector=&MergeHelper_VectorSwitch(0=<0 -300 0>, 1=<0 -300 0>, 2=<0 -300 0>, 3=<0 -900 0>, Index=$User.GameModeID, Num=::NewEnumerator2)
     Passthrough Vector=&MergeHelper_VectorSwitch(0=<0 300 0>, 1=<0 300 0>, 2=<0 300 0>, 3=<0 900 0>, Index=$User.GameModeID, Num=::NewEnumerator2)
    ))
   &DLR_ProductPreviews(GameModeID:NiagaraInt32=$System.GameModeID, bPreviewVehicleSpeed:NiagaraBool=$System.bPreviewVehicleSpeed)
   /SetVariables(Particles.Position=&AddVectorToPosition(
     Position=$Particles.Position
     Vector=&TransformVector(
      Vector=&MakeVector(
       X=0
       Y=&MaskFloat(
        Mask Source=0
        Mask Parameter=?
        Mask=0
        Masked Float=&MergeHelper_FloatSwitch(0=0, 1=0, 2=0, 3=0, Index=$User.GameModeID, Num=::NewEnumerator2)
        Passthrough Float=&MergeHelper_FloatSwitch(0=0, 1=0, 2=0, 3=0, Index=$User.GameModeID, Num=::NewEnumerator2)
       )
       Z=0
      )
      Destination Space=::Simulation
     )
    )) {enabled=false, isSetParameters=true}
  ]
  particleUpdateScript [
   &ParticleState()
   &InheritSourceMovement(
    Applied Position Scale=&Multiply_VectorByFloat(
     Vector=<1 1 1>
     Float=&FloatFromCurve(FloatCurve=DataInterface() {propertyValues {Curve {keys [{interpMode=RCIM_Cubic, tangentMode=RCTM_Break, time=0.3, value=1}, {interpMode=RCIM_Cubic, tangentMode=RCTM_Break, time=1, value=0}]}}})
    )
    ApplyForce=false
    ApplyPosition=true
    ApplyPositionFalloff=false
   ) {enabled=false, module=InheritSourceMovement001}
   &InheritSourceMovement(
    Applied Position Scale=&VectorFromFloat(Value=&FloatFromCurve(FloatCurve=DataInterface() {propertyValues {Curve {
         defaultValue=1
         keys [
          {arriveTangent=-1.7885, interpMode=RCIM_Cubic, leaveTangent=-1.7885, tangentMode=RCTM_User, time=0.05, value=1}
          {interpMode=RCIM_Cubic, tangentMode=RCTM_User, time=1, value=0}
         ]
        }}}))
    ApplyForce=false
    ApplyPosition=true
    ApplyPositionFalloff=false
   )
   &SolveForcesAndVelocity(Clamp Velocity=false, Limit Acceleration=false)
   &ScaleColor(Scale RGB=<1.45 1.45 1.45>, ScaleA=true, ScaleRGB=true)
   &DynamicMaterialParameters(
    Index 0 Param 1=2.5
    Index 0 Param 2=-4.69
    Index 0 Param 3=-2.75
    Module.Index0Param0Write=true
    Module.Index0Param1Write=true
    Module.Index0Param2Write=true
    Module.Index0Param3Write=true
   )
   &ScaleRibbonWidth(Ribbon Width Scale=&FloatFromCurve(FloatCurve=DataInterface() {propertyValues {Curve {keys [
         {interpMode=RCIM_Cubic, tangentMode=RCTM_Break, time=0, value=0.4}
         {interpMode=RCIM_Cubic, tangentMode=RCTM_Break, time=0.08, value=1}
         {interpMode=RCIM_Cubic, tangentMode=RCTM_Break, time=0.42, value=0.52}
         {interpMode=RCIM_Cubic, tangentMode=RCTM_Break, time=0.76, value=0.89}
         {interpMode=RCIM_Cubic, tangentMode=RCTM_Break, time=1.14, value=1.06}
         {interpMode=RCIM_Cubic, tangentMode=RCTM_Break, time=1.25, value=1.66}
        ]}}}))
   &VC_UpdateRibbon_v1()
  ]
  rendererClasses [&NiagaraRibbonRendererProperties]
  renderers [{rendererClass=&NiagaraRibbonRendererProperties}]
 }
}
summary {
 systemName=NS_Sticky
 emitters [
  {emitterName=RibbonN001, rendererClasses [&NiagaraRibbonRendererProperties]}
  {emitterName=Optimized_Trail, rendererClasses [&NiagaraRibbonRendererProperties]}
  {emitterName=ConeOptimized, rendererClasses [&NiagaraMeshRendererProperties, &NiagaraSpriteRendererProperties]}
 ]
 userVariables [
  User.bIsFrontendPreview {type=NiagaraBool}
  User.bIsPainted {type=NiagaraBool}
  User.blsIsCoroColor {type=NiagaraBool}
  User.bPreviewVehicleSpeed {type=NiagaraBool}
  User.ColorVariantID {type=NiagaraInt32}
  User.DTCarOffset {type=NiagaraPosition, description="", defaultValue {}}
  User.GameModeID {type=NiagaraInt32, defaultValue=3}
  User.PaintedColor {type=LinearColor, description="", defaultValue {struct=LinearColor, value=<rgba 1 1 1 1>}}
  User.PaintedScalar {type=NiagaraFloat, defaultValue=1, description=""}
  User.PortCount {type=NiagaraInt32, defaultValue=1}
  User.Power_AlphaPaintedR1 {type=NiagaraFloat, defaultValue=1, description=""}
  User.Power_EmissivePaintedR1 {type=NiagaraFloat, defaultValue=1, description=""}
  User.SingleBoostPos {type=NiagaraPosition, description="", defaultValue {}}
  User.SingleBoostPos {type=Vector3f, description="", defaultValue {struct=Vector3f, value=<0 0 0>}}
 ]
}
