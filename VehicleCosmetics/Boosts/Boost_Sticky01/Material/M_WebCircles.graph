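# Translucent, unlit, two-sided material for Niagara mesh particles in the Sticky01 vehicle boost effect, drawing dissolving web circles.
# A static switch picks radially mapped panning noise rings (default) or edge-masked scrolling noise, smoothstepped by particle Dissolve and scaled by particle alpha.
# An optional switch multiplies in a web-noise mask.
# Noise scaled by Glow feeds an emissive adjustment function, while layered time-driven sines scaled by Distort compute a two-axis wobble.
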Material /VehicleCosmetics/Boosts/Boost_Sticky01/Material/M_WebCircles
graph {
 properties {
  bAutomaticallySetUsageInEditor=false
  bUsedWithNiagaraMeshParticles=true
  blendMode=BLEND_Translucent
  shadingModel=MSM_Unlit
  twoSided=true
  usedShadingModels=Unlit
 }
 nodes [
  n1 = TextureCoordinate() {coordinateIndex=1}
  MaterialExpressionConstantClamp_1 = Saturate(None=ConstantBiasScale(None=ComponentMask(None=@n1) {g=true}) {bias=-0.01, scale=4})
  n2 = TextureCoordinate()
  n3 = AppendVector(A=ScalarParameter() {defaultValue=1, parameterName=TileX}, B=ScalarParameter() {defaultValue=1, parameterName=TileY})
  n4 = Multiply(A=@n2, B=@n3)
  n5 = Multiply(
   A=@MaterialExpressionConstantClamp_1
   B=OneMinus(None=TextureSample(UVs=@n4) {texture=/VehicleCosmetics/Boosts/Boost_Sticky01/Textures/T_VCVFX_Boost_STicky01_NoiseWeb_A}.RGB)
  )
  MaterialExpressionMeshEmitterVertexColor_0 = ParticleColor()
  MaterialExpressionMeshEmitterDynamicParameter_0 = DynamicParameter() {paramNames [Dissolve,Glow,Distort,Param4]}
  n6 = ScalarParameter() {parameterName=Max}
  n7 = Constant2Vector() {g=0.5, r=0.5}
  n8 = TextureCoordinate()
  n9 = ScalarParameter() {defaultValue=0.525, parameterName=Radius}
  n10 = ScalarParameter() {defaultValue=2.25, parameterName=Density}
  n11 = Multiply(
   A=MaterialFunctionCall() {materialFunction=&VectorToRadialValue}.Radial Coordinates
   B=@n3
  )
  n12 = Multiply(
   A=Sine(None=MaterialFunctionCall(UVs=@n7, CenterPosition=@n8, Radius=@n9, Density=@n10) {materialFunction=&RadialGradientExponential}.RadialGradientExponential) {period=2}
   B=TextureSample(UVs=Panner(Coordinate=@n11) {speedY=-0.3}) {texture=/VehicleCosmetics/Boosts/Boost_Sticky01/Textures/T_VCVFX_Boost_STicky01_DissolveFire_A}.RGB
  )
  n13 = Power(
   Base=TextureSample(UVs=@n2) {texture=/VehicleCosmetics/Boosts/Boost_Sticky01/Textures/T_VCVFX_Boost_STicky01_CautionTapeEdgeMask_A}.RGB
   Exp=ScalarParameter() {defaultValue=1.5, parameterName=PowerEdgeMask}
  )
  n14 = ScalarParameter() {defaultValue=2, parameterName=MultEdgeMask}
  MaterialExpressionConstantClamp_3 = Saturate(None=Multiply(
    A=TextureSample(UVs=@n2) {texture=/VehicleCosmetics/Boosts/Boost_Sticky01/Textures/T_VCVFX_Boost_STicky01_GradientStripeVertical_A}.RGB
    B=Multiply(A=@n13, B=@n14)
   ))
  n15 = Time()
  n16 = ScalarParameter() {defaultValue=0.5, parameterName=SpeedNoise}
  n17 = TextureSampleParameter2D(UVs=Panner(Coordinate=@n4, Time=Multiply(A=@n15, B=@n16)) {speedY=0.1}) {parameterName=DissolveText, texture=&T_VCVFX_Boost_STicky01_DissolveFire_A}
  n18 = Multiply(A=@MaterialExpressionConstantClamp_3, B=@n17.R)
  MaterialExpressionConstantClamp_2 = Saturate(None=SmoothStep(
    Min=@MaterialExpressionMeshEmitterDynamicParameter_0.Dissolve
    Max=Add(A=@MaterialExpressionMeshEmitterDynamicParameter_0.Dissolve, B=@n6)
    Value=StaticSwitchParameter(True=@n12, False=@n18) {defaultValue=true, parameterName=Radial}
   ))
  n19 = Multiply(A=@MaterialExpressionMeshEmitterVertexColor_0.A, B=@MaterialExpressionConstantClamp_2)
  MaterialExpressionConstantClamp_0 = Saturate(None=StaticSwitchParameter(True=Multiply(A=@n5, B=@n19), False=@n19) {parameterName=Comb})
  MaterialFunctionCall(Emissive=Multiply(A=@n17.G, B=@MaterialExpressionMeshEmitterDynamicParameter_0.Glow)) {materialFunction=&MF_FX_EmissiveAdjustments}
  n20 = Divide(A=ComponentMask(None=TextureCoordinate() {coordinateIndex=1}) {r=true}, B=ScalarParameter() {defaultValue=0.8, parameterName=CountWaves2})
  n21 = Multiply(A=ScalarParameter() {defaultValue=0.5, parameterName=HorizontalSpeed}, B=OneMinus(None=Time()))
  n22 = Multiply(A=ScalarParameter() {defaultValue=-0.2, parameterName=HorizontalSpeed2}, B=OneMinus(None=Time()))
  n23 = Multiply(A=Sine(None=Add(A=@n20, B=@n21)), B=Sine(None=Add(B=@n22)))
  n24 = Constant() {r=1.2}
  n25 = TextureCoordinate() {coordinateIndex=1}
  n26 = Time()
  n27 = Add(
   A=Divide(A=ComponentMask(None=@n25) {g=true}, B=ScalarParameter() {defaultValue=0.8, parameterName=CountWaves2})
   B=Multiply(A=ScalarParameter() {defaultValue=1, parameterName=VerticalSpeed}, B=OneMinus(None=@n26))
  )
  n28 = Time()
  n29 = Add(B=Multiply(A=ScalarParameter() {defaultValue=-0.5, parameterName=VerticalSpeed2}, B=OneMinus(None=@n28)))
  n30 = Multiply(A=OneMinus(None=Sine(None=@n27)), B=OneMinus(None=Sine(None=@n29)))
  n31 = Constant() {r=1.2}
  Multiply(A=AppendVector(A=Power(Base=@n23, Exp=@n24), B=Power(Base=@n30, Exp=@n31)), B=@MaterialExpressionMeshEmitterDynamicParameter_0.Distort)
 ]
}
