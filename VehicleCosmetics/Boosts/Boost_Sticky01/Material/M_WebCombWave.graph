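# Translucent, unlit, two-sided material for Niagara mesh particles that draws a waving web effect for the Sticky01 vehicle boost.
# A panning mask distorts the web texture, which combines with particle color to build emissive and opacity.
# The Dissolve dynamic parameter erodes opacity through a smoothstep over radial fire noise, or edge-masked panning noise via static switch.
# Time-driven sine waves form a vertex offset, weighted along V to hold one end nearly still.
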Material /VehicleCosmetics/Boosts/Boost_Sticky01/Material/M_WebCombWave
graph {
 properties {
  bAutomaticallySetUsageInEditor=false
  bUsedWithNiagaraMeshParticles=true
  blendMode=BLEND_Translucent
  shadingModel=MSM_Unlit
  twoSided=true
  usedShadingModels=Unlit
 }
 nodes [
  n1 = Multiply(
   A=TextureCoordinate()
   B=AppendVector(A=ScalarParameter() {defaultValue=1, parameterName=TilingTexture_X}, B=ScalarParameter() {defaultValue=2, parameterName=TilingTexture_Y})
  )
  n2 = ScalarParameter() {defaultValue=1, parameterName=Distort_X}
  n3 = ScalarParameter() {defaultValue=1, parameterName=Distort_Y}
  n4 = Panner(Coordinate=Multiply(A=TextureCoordinate(), B=AppendVector(A=@n2, B=@n3))) {speedX=-0.8, speedY=0.3}
  n5 = Multiply(
   A=ComponentMask(None=TextureSample(UVs=@n4) {texture=/VehicleCosmetics/Boosts/Boost_Sticky01/Textures/TrailMask_RGB_T}.RGB) {g=true, r=true}
   B=ScalarParameter() {defaultValue=0.3, parameterName=Distortion}
  )
  n6 = TextureSampleParameter2D(UVs=StaticSwitchParameter(True=@n1, False=Add(A=@n1, B=@n5))) {parameterName=WebTexture, texture=/VehicleCosmetics/Boosts/Boost_Sticky01/Textures/T_VCVFX_Boost_STicky01_WebCone2_A}
  n7 = ParticleColor()
  n8 = Multiply(A=@n6.A, B=@n7.A)
  n9 = TextureCoordinate()
  n10 = AppendVector(A=ScalarParameter() {defaultValue=2.2, parameterName=TilingTexture_X}, B=Constant() {r=1})
  n11 = OneMinus(None=TextureSampleParameter2D(UVs=Multiply(A=@n9, B=@n10)) {parameterName=WebTexture, texture=&T_VCVFX_Boost_STicky01_WebCone2_A}.B)
  MaterialExpressionMeshEmitterDynamicParameter_1 = DynamicParameter() {paramNames [Dissolve,Glow,Distort,Param4]}
  n12 = ScalarParameter() {parameterName=Max}
  n13 = Constant2Vector() {g=0.5, r=0.5}
  n14 = TextureCoordinate()
  n15 = ScalarParameter() {defaultValue=0.525, parameterName=Radius}
  n16 = ScalarParameter() {defaultValue=2.25, parameterName=Density}
  n17 = AppendVector(A=ScalarParameter() {defaultValue=1, parameterName=TileX}, B=ScalarParameter() {defaultValue=1, parameterName=TileY})
  n18 = Multiply(
   A=MaterialFunctionCall() {materialFunction=&VectorToRadialValue}.Radial Coordinates
   B=@n17
  )
  n19 = Multiply(
   A=Sine(None=MaterialFunctionCall(UVs=@n13, CenterPosition=@n14, Radius=@n15, Density=@n16) {materialFunction=&RadialGradientExponential}.RadialGradientExponential) {period=2}
   B=TextureSample(UVs=Panner(Coordinate=@n18) {speedY=-0.3}) {texture=/VehicleCosmetics/Boosts/Boost_Sticky01/Textures/T_VCVFX_Boost_STicky01_DissolveFire_A}.RGB
  )
  n20 = TextureCoordinate()
  n21 = Power(
   Base=TextureSample(UVs=@n20) {texture=/VehicleCosmetics/Boosts/Boost_Sticky01/Textures/T_VCVFX_Boost_STicky01_CautionTapeEdgeMask_A}.RGB
   Exp=ScalarParameter() {defaultValue=1.5, parameterName=PowerEdgeMask}
  )
  n22 = ScalarParameter() {defaultValue=2, parameterName=MultEdgeMask}
  MaterialExpressionConstantClamp_6 = Saturate(None=Multiply(
    A=TextureSample(UVs=@n20) {texture=/VehicleCosmetics/Boosts/Boost_Sticky01/Textures/T_VCVFX_Boost_STicky01_GradientStripeVertical_A}.RGB
    B=Multiply(A=@n21, B=@n22)
   ))
  n23 = Multiply(A=@n17, B=@n20)
  n24 = Multiply(A=Time(), B=ScalarParameter() {defaultValue=0.5, parameterName=SpeedNoise})
  n25 = Multiply(
   A=@MaterialExpressionConstantClamp_6
   B=TextureSampleParameter2D(UVs=Panner(Coordinate=@n23, Time=@n24) {speedY=0.1}) {parameterName=DissolveText, texture=&T_VCVFX_Boost_STicky01_DissolveFire_A}.R
  )
  MaterialExpressionConstantClamp_5 = Saturate(None=SmoothStep(
    Min=@MaterialExpressionMeshEmitterDynamicParameter_1.Dissolve
    Max=Add(A=@MaterialExpressionMeshEmitterDynamicParameter_1.Dissolve, B=@n12)
    Value=StaticSwitchParameter(True=@n19, False=@n25) {defaultValue=true, parameterName=Radial}
   ))
  MaterialExpressionConstantClamp_4 = Saturate(None=Multiply(A=Multiply(A=@n8, B=@n11), B=@MaterialExpressionConstantClamp_5))
  n26 = Multiply(A=@n6.R, B=ScalarParameter() {defaultValue=1, parameterName=Glow})
  MaterialFunctionCall(Emissive=Multiply(A=Multiply(A=@n26, B=@n11), B=@n7.RGB)) {materialFunction=&MF_FX_EmissiveAdjustments}
  TextureCoordinate()
  n27 = TextureCoordinate()
  n28 = Divide(A=OneMinus(None=ComponentMask(None=@n27) {g=true}), B=ScalarParameter() {defaultValue=1, group=WPO, parameterName=tile})
  n29 = Multiply(A=OneMinus(None=Time()), B=ScalarParameter() {defaultValue=2.5, group=WPO, parameterName=Speed})
  n30 = AppendVector(A=Constant(), B=Sine(None=Add(A=@n28, B=@n29)) {period=0.2})
  n31 = Cosine(None=Add()) {period=0.7}
  n32 = Multiply(
   A=Multiply(A=AppendVector(A=@n30, B=@n31), B=ScalarParameter() {defaultValue=20, group=WPO, parameterName=Intensity})
   B=ScalarParameter() {defaultValue=1, group=WPO, parameterName=Amount}
  )
  n33 = TextureCoordinate()
  n34 = Power(Base=OneMinus(None=ComponentMask(None=@n33) {g=true}), Exp=ScalarParameter() {defaultValue=1.5, parameterName=PowerMiddle2})
  n35 = ScalarParameter() {defaultValue=0.05, parameterName=StaticPoint}
  Transform(None=Multiply(A=ConstantBiasScale(None=@n32), B=Subtract(A=@n34, B=@n35)))
 ]
}
