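# Translucent, unlit, two-sided material for Niagara ribbons that draws waving web strands for the Sticky01 vehicle boost.
# Scrolling web and noise-distorted strip textures, with tiling and pan speed from dynamic parameters, combine with panning noise masks and fades along U to form opacity.
# Particle color tints texture channels for emissive via MF_FX_EmissiveAdjustments.
# Time-driven sine and cosine waves feed world position offset, weighted along U so displacement grows along the ribbon.
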Material /VehicleCosmetics/Boosts/Boost_Sticky01/Material/M_WebWave
graph {
 properties {
  bAutomaticallySetUsageInEditor=false
  bUsedWithNiagaraRibbons=true
  blendMode=BLEND_Translucent
  shadingModel=MSM_Unlit
  twoSided=true
  usedShadingModels=Unlit
 }
 nodes [
  n1 = TextureCoordinate()
  n2 = AppendVector(A=ScalarParameter() {defaultValue=1, parameterName=Distort_X}, B=ScalarParameter() {defaultValue=1, parameterName=Distort_Y})
  n3 = TextureSample(UVs=Panner(Coordinate=Multiply(A=@n1, B=@n2)) {speedX=-0.8, speedY=0.3}) {texture=/DealershipSharedLibrary/FX/Textures/Noise/T_VCVFX_TrailMask_RGB}
  n4 = Multiply(
   A=Multiply(A=ComponentMask(None=@n3.RGB) {g=true, r=true}, B=ScalarParameter() {defaultValue=0.3, parameterName=Distortion})
   B=ScalarParameter() {defaultValue=0.2, parameterName=Strips}
  )
  n5 = Multiply(A=TextureCoordinate(), B=AppendVector(A=Constant() {r=1}, B=Constant() {r=1}))
  n6 = Time()
  n7 = DynamicParameter() {paramNames [TilingX,Panner01,Panner02,Param4]}
  n8 = TextureSampleParameter2D(UVs=Panner(Coordinate=Add(A=@n4, B=@n5), Time=Multiply(A=@n6, B=@n7.Panner02)) {speedX=0.2}) {parameterName=Line, texture=/VehicleCosmetics/Boosts/Boost_Sticky01/Textures/T_VCVFX_Boost_STicky01_WebStrips_A}
  n9 = Multiply(A=@n8.B, B=Constant() {r=0.2})
  n10 = ComponentMask(None=TextureCoordinate()) {r=true}
  n11 = Power(Base=ConstantBiasScale(None=@n10) {bias=-0.001, scale=4}, Exp=ScalarParameter() {defaultValue=0.3, parameterName=Fade})
  n12 = DynamicParameter() {paramNames [TilingX,Panner01,Panner02,Param4]}
  n13 = Multiply(A=TextureCoordinate(), B=AppendVector(A=@n12.TilingX, B=ScalarParameter() {defaultValue=1, parameterName=TilingTexture_Y}))
  n14 = Time()
  n15 = Multiply(A=@n12.Panner01, B=ScalarParameter() {defaultValue=3, parameterName=MultiWebPanner})
  n16 = TextureSampleParameter2D(UVs=Panner(Coordinate=Add(A=@n13), Time=Multiply(A=@n14, B=@n15)) {speedX=0.2}) {parameterName=Texture, texture=/VehicleCosmetics/Boosts/Boost_Sticky01/Textures/T_VCVFX_Boost_STicky01_StripWeb02_A}
  MaterialExpressionVertexColor_0 = ParticleColor()
  n17 = Multiply(A=@n16.R, B=@MaterialExpressionVertexColor_0.A)
  MaterialExpressionConstantClamp_12 = Saturate(None=Add(A=@n9, B=Multiply(A=@n17, B=@n11)))
  n18 = TextureCoordinate()
  n19 = Multiply(A=Power(Base=ComponentMask(None=@n18) {r=true}, Exp=Constant() {r=0.5}), B=Constant() {r=50})
  n20 = TextureCoordinate()
  n21 = Power(Base=OneMinus(None=ComponentMask(None=@n20) {r=true}), Exp=Constant() {r=5})
  n22 = Multiply(A=@n21, B=Constant() {r=10})
  n23 = ScalarParameter() {defaultValue=1, parameterName=MaskSpeed}
  n24 = Time()
  n25 = TextureSample(UVs=Panner(Coordinate=TextureCoordinate() {uTiling=0.1, vTiling=0.65}, Time=Multiply(A=@n23, B=@n24)) {speedX=-1, speedY=-0.1}) {texture=/VehicleCosmetics/Boosts/Boost_Sticky01/Textures/T_VCVFX_Boost_STicky01_NoiseWeb01_A}
  n26 = Constant() {r=3}
  n27 = TextureSample(UVs=Panner(Coordinate=TextureCoordinate() {uTiling=0.5, vTiling=0.5}) {speedX=-1, speedY=-0.1}) {texture=&T_VCVFX_Boost_STicky01_NoiseWeb01_A}
  n28 = Constant() {r=1.5}
  MaterialExpressionConstantClamp_10 = Saturate(None=Add(A=Multiply(A=@n25.RGB, B=@n26), B=Multiply(A=@n27.RGB, B=@n28)))
  n29 = Add(A=@n22, B=@MaterialExpressionConstantClamp_10)
  MaterialExpressionConstantClamp_9 = Saturate(None=Add(A=@n21, B=Multiply(A=@n22, B=@n29)))
  n30 = Add(A=Multiply(A=@n9, B=@n11), B=Multiply(A=@MaterialExpressionConstantClamp_12, B=Multiply(A=@n19, B=@MaterialExpressionConstantClamp_9)))
  n31 = ConstantBiasScale(None=OneMinus(None=@n10)) {bias=-0.001, scale=4}
  MaterialExpressionConstantClamp_11 = Saturate(None=Multiply(A=Multiply(A=@n30, B=@n31), B=@n11))
  n32 = Add(A=@n16.G, B=@n8.R)
  MaterialFunctionCall(Emissive=Add(A=Multiply(A=@MaterialExpressionVertexColor_0.RGB, B=@n32)) {constB=0.1}) {materialFunction=&MF_FX_EmissiveAdjustments}
  n33 = Multiply(A=OneMinus(None=Time()), B=ScalarParameter() {defaultValue=2.5, group=WPO, parameterName=Speed})
  n34 = Divide(A=ComponentMask(None=TextureCoordinate()) {r=true}, B=ScalarParameter() {defaultValue=1, group=WPO, parameterName=tile})
  n35 = Add(A=@n33, B=@n34)
  n36 = Add(A=@n34, B=@n33)
  n37 = AppendVector(A=AppendVector(A=Cosine(None=@n35) {period=0.7}, B=Sine(None=@n36) {period=0.2}), B=Constant())
  n38 = ScalarParameter() {defaultValue=20, group=WPO, parameterName=Intensity}
  n39 = ConstantBiasScale(None=Multiply(A=Multiply(A=@n37, B=@n38), B=ScalarParameter() {defaultValue=1, group=WPO, parameterName=Amount}))
  n40 = TextureCoordinate()
  n41 = Subtract(
   A=Power(Base=ComponentMask(None=@n40) {r=true}, Exp=ScalarParameter() {defaultValue=1.5, parameterName=PowerMiddle2})
   B=ScalarParameter() {defaultValue=0.05, parameterName=StaticPoint}
  )
  Multiply(A=Transform(None=Multiply(A=@n39, B=@n41)), B=ScalarParameter() {defaultValue=2, parameterName=MultiREsultWPO})
 ]
}
